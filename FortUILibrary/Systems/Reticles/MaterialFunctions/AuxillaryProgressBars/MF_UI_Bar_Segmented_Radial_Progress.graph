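# UI material function for reticle auxiliary bars that draws a segmented radial progress meter along an arc, with optional snapping to whole segments.
# Radial arc fills build the progress mask and a delta fill spanning up to one segment past it, which drops out when Delta Progress reaches one.
# Both combine with a dimmed backing, then get cut by anti-aliased SDF sectors and a circle outline into the opacity mask, alongside separate progress mask outputs.
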MaterialFunction /FortUILibrary/Systems/Reticles/MaterialFunctions/AuxillaryProgressBars/MF_UI_Bar_Segmented_Radial_Progress
graph {
 properties {
  bExposeToLibrary=true
  description=Creates a Segmented Progress Bar with inputs for snapping the segments or not.
  previewBlendMode=BLEND_Translucent
  previewMaterialDomain=MD_UI
 }
 nodes [
  FunctionOutput(None=NamedRerouteUsage()) {outputName=Delta Progress Mask, sortPriority=1}
  n1 = FunctionInput(Preview=MaterialFunctionCall() {materialFunction=&GetUserInterfaceUV}.Normalized UV) {bUsePreviewValueAsDefault=true, inputName=UVs, inputType=FunctionInput_Vector2, sortPriority=-2}
  n2 = FunctionInput() {
   bUsePreviewValueAsDefault=true
   description=Defines artc start angle of meter
   inputName="Start Angle (0-1)"
   inputType=FunctionInput_Scalar
   sortPriority=4
   previewValue=<0.25 0 0 1>
  }
  n3 = FunctionInput() {
   bUsePreviewValueAsDefault=true
   description=Define arc end angle or meter
   inputName="End Angle (0-1)"
   inputType=FunctionInput_Scalar
   sortPriority=5
   previewValue=<0.5 0 0 1>
  }
  n4 = FunctionInput() {
   bUsePreviewValueAsDefault=true
   description=Meter Progress
   inputName=Meter Progress
   inputType=FunctionInput_Scalar
   sortPriority=-1
   previewValue=<0.5 0 0 1>
  }
  n5 = FunctionInput() {
   bUsePreviewValueAsDefault=true
   description=Defines the amount of segments. Used for defining clip size of the ability/weapon.
   inputName=Segment Count
   inputType=FunctionInput_Scalar
   sortPriority=1
   previewValue=<20 0 0 1>
  }
  n6 = Floor(None=@n5)
  n7 = FunctionInput(Preview=StaticBool()) {
   bUsePreviewValueAsDefault=true
   description=Rounds Fill down to the nearest segment
   inputName=Snap Progress to Segments
   inputType=FunctionInput_StaticBool
   sortPriority=50
  }
  n8 = StaticSwitch(
   True=MaterialFunctionCall!("Value (0-1)"=@n4, Steps=@n6) {materialFunction=/FortUILibrary/MaterialFunctions/Basics/MF_UI_StepRatio}.Stepped Ratio
   False=@n4
   Value=@n7
  )
  n9 = OneMinus(None=@n8)
  n10 = FunctionInput() {
   bUsePreviewValueAsDefault=true
   description=Changes the origin point of the fill.
   inputName=Fill Direction
   inputType=FunctionInput_Scalar
   sortPriority=6
  }
  n11 = Constant() {r=0.001}
  n12 = NamedRerouteDeclaration(None=MaterialFunctionCall(
    DistanceField=MaterialFunctionCall!(UV=@n1, "Arc Start Angle (0-1)"=@n2, "Arc End Angle (0-1)"=@n3, "Fill Ratio (0-1)"=@n9, Fill Clockwise=@n10) {materialFunction=/FortUILibrary/MaterialFunctions/Basics/MF_UI_RadialArc}.Radial Fill
    Threshold=@n11
   ) {materialFunction=&MF_UI_AntiAliasing}.Outer) {name=SegmentedMeterFill, nodeColor=<rgba 0.025 0.0203 0.0125 0.75>}
  n13 = FunctionInput() {bUsePreviewValueAsDefault=true, description=Defines Progress of Delta Fill, inputName=Delta Progress, inputType=FunctionInput_Scalar}
  n14 = Constant() {r=1}
  n15 = Add(
   A=@n8
   B=MaterialFunctionCall(Input=@n13, Target High=Divide(A=@n14, B=@n6)) {materialFunction=&RemapValueRange}.Result
  )
  n16 = LinearInterpolate(Alpha=Floor(None=@n13)) {constB=-1}
  n17 = MaterialFunctionCall(UV=@n1, "Arc Start Angle (0-1)"=@n2, "Arc End Angle (0-1)"=@n3, "Fill Ratio (0-1)"=OneMinus(None=Add(A=@n15, B=@n16)), Fill Clockwise=@n10) {materialFunction=&MF_UI_RadialArc}
  n18 = Subtract(
   A=MaterialFunctionCall(DistanceField=@n17.Radial Fill, Threshold=@n11) {materialFunction=&MF_UI_AntiAliasing}.Outer
   B=@n12
  )
  n19 = OneMinus(None=@n17.Arc Mask)
  n20 = FunctionInput() {
   bUsePreviewValueAsDefault=true
   description=The opacity of the backing element for the meter
   inputName="Background Opacity (0-1)"
   inputType=FunctionInput_Scalar
   sortPriority=7
   previewValue=<0.5 0 0 1>
  }
  n21 = Max(
   A=Max(A=@n12, B=NamedRerouteDeclaration(None=@n18) {name=SegmentedDeltaFill, nodeColor=<rgba 0.025 0.0203 0.0125 0.75>})
   B=Multiply(A=Ceil(None=@n19), B=@n20) {constB=0.1}
  )
  n22 = FunctionInput() {
   bUsePreviewValueAsDefault=true
   description=Defines spacing between sectors
   inputName=Segment Spacing
   inputType=FunctionInput_Scalar
   sortPriority=2
   previewValue=<0.25 0 0 1>
  }
  n23 = OneMinus(None=@n22)
  n24 = FunctionInput() {
   bUsePreviewValueAsDefault=true
   description=Define if the spacing is fixed
   inputName=Fixed Spacing
   inputType=FunctionInput_Scalar
   sortPriority=49
  }
  n25 = MaterialFunctionCall(
   DistanceField=MaterialFunctionCall!(
    UVs=@n1
    Gradient Map=@n17.Arc Mask
    Sector Count=@n6
    Sector Spacing=NamedRerouteDeclaration(None=@n23) {name=Segment Spacing, nodeColor=<rgba 0.025 0.0203 0.0125 0.75>}
    Fixed Spacing=@n24
   ) {materialFunction=/FortUILibrary/MaterialFunctions/Basics/MF_UI_Sectors}.Parallel Sectors SDF
   Threshold=Constant() {r=0.25}
  ) {materialFunction=&MF_UI_AntiAliasing}
  n26 = FunctionInput() {
   bUsePreviewValueAsDefault=true
   description=Defines the size of the circle mask where 1 is the edge of the border.
   inputName="Circle Size (0-1)"
   inputType=FunctionInput_Scalar
   sortPriority=8
   previewValue=<0.75 0 0 1>
  }
  n27 = FunctionInput() {
   bUsePreviewValueAsDefault=true
   description=Defubes the stroke thickness of the circle mask.
   inputName=Outline Thickness
   inputType=FunctionInput_Scalar
   sortPriority=9
   previewValue=<0.05 0 0 1>
  }
  n28 = MaterialFunctionCall(Size=@n26, OutlineThickness=@n27) {materialFunction=&MF_UI_Circle}
  FunctionOutput(None=Multiply(
    A=MaterialFunctionCall(DistanceField=@n28.SDFOutline) {materialFunction=&MF_UI_AntiAliasing}.Inner
    B=Multiply(A=@n21, B=@n25.Inner)
   )) {outputName=Opacity Mask, sortPriority=-1}
  FunctionOutput(None=NamedRerouteUsage()) {description="The combined fill+backing meter. ", outputName=Meter Progress Mask}
 ]
}
